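# Animation Blueprint for the AgarBelly guitar cosmetic that adds sprung, angle-limited Anim Dynamics motion to its left and right bell bones.
# Three four-bone chains share the cached input pose, varying in strength and tuning, and nested bool blends select the output.
# A separate two-bone chain does not reach the output.
# The event graph flips gravity_override upward while bIsSkydiveFloating is true, but no node visibly reads that variable.
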AnimBlueprint /SparksCosmetics/Instrument/Guitar/AgarBelly/Meshes/AgarBelly_Guitar_AnimBP
target_skeleton=&/AgarBelly_Guitar_Skeleton
graphs [
 AnimGraph {nodes [
   n1 = Anim Dynamics - Bone: dyn_bell_lf!(ComponentPose, Alpha=1, GravityOverride:Vector="0,0,-980", ExternalForce="0,0,0") -> (Pose) {
    angularDampingOverride=1
    angularSpringConstant=20000
    bAngularSpring=true
    bOverrideAngularDamping=true
    bUseGravityOverride=true
    gravityScale=0.35
    simulationSpace=RootRelative
    boundBone {boneName=dyn_bell_lf}
    componentLinearAccScale=<1 1 1>
    componentLinearVelScale=<1 1 1>
    gravityOverride {z=-980}
    physicsBodyDefinitions [{
      boundBone {boneName=dyn_bell_lf}
      boxExtents=<20 30 30>
      constraintSetup {angularLimitsMax=<15 15 15>, angularLimitsMin=<-15 -15 -15>, angularTarget=<1 0 0>}
      localJointOffset=<3.5 0 0>
     }]
    simSpaceSettings {simSpaceAngularAlpha=1}
   }
   Anim Dynamics - Bone: dyn_bell_rt!(ComponentPose=@n1, Alpha=1, GravityOverride:Vector="0,0,-980", ExternalForce="0,0,0") -> (Pose) {
    angularDampingOverride=1
    angularSpringConstant=20000
    bAngularSpring=true
    bOverrideAngularDamping=true
    bUseGravityOverride=true
    gravityScale=0.35
    simulationSpace=RootRelative
    boundBone {boneName=dyn_bell_rt}
    componentLinearAccScale=<1 1 1>
    componentLinearVelScale=<1 1 1>
    gravityOverride {z=-980}
    physicsBodyDefinitions [{
      boundBone {boneName=dyn_bell_rt}
      boxExtents=<20 30 30>
      constraintSetup {angularLimitsMax=<15 15 15>, angularLimitsMin=<-15 -15 -15>, angularTarget=<1 0 0>}
      localJointOffset=<3.5 0 0>
     }]
    simSpaceSettings {simSpaceAngularAlpha=1}
   }
   n2 = Anim Dynamics - Bone: dyn_bell_01_rt!(ComponentPose=Local To Component(LocalPose=Use cached pose 'Input_Pose'()), ExternalForce="0,0,0") -> (Pose) {
    alpha=0.7
    angularDampingOverride=1
    angularSpringConstant=60000
    bAngularSpring=true
    bOverrideAngularDamping=true
    bUseGravityOverride=true
    gravityScale=0.35
    simulationSpace=RootRelative
    boundBone {boneName=dyn_bell_01_rt}
    componentAppliedLinearAccClamp=<0 0 0>
    gravityOverride {z=-980}
    physicsBodyDefinitions [{
      boundBone {boneName=dyn_bell_01_rt}
      boxExtents=<15 25 25>
      constraintSetup {angularLimitsMax=<0 0 5>, angularLimitsMin=<0 0 -5>, angularTarget=<1 0 0>}
      localJointOffset=<1 0 0>
     }]
    simSpaceSettings {simSpaceAngularAlpha=1}
   }
   n3 = Anim Dynamics - Bone: dyn_bell_02_rt!(ComponentPose=@n2, ExternalForce="0,0,0") -> (Pose) {
    alpha=0.7
    angularDampingOverride=1
    angularSpringConstant=60000
    bAngularSpring=true
    bOverrideAngularDamping=true
    bUseGravityOverride=true
    gravityScale=0.35
    simulationSpace=RootRelative
    boundBone {boneName=dyn_bell_02_rt}
    componentAppliedLinearAccClamp=<0 0 0>
    gravityOverride {z=-980}
    physicsBodyDefinitions [{
      boundBone {boneName=dyn_bell_02_rt}
      boxExtents=<15 25 25>
      constraintSetup {angularLimitsMax=<0 0 10>, angularLimitsMin=<0 0 -10>, angularTarget=<1 0 0>}
      localJointOffset=<1 0 0>
     }]
    simSpaceSettings {simSpaceAngularAlpha=1}
   }
   n4 = Anim Dynamics - Bone: dyn_bell_01_lf!(ComponentPose=@n3, ExternalForce="0,0,0") -> (Pose) {
    alpha=0.7
    angularDampingOverride=1
    angularSpringConstant=60000
    bAngularSpring=true
    bOverrideAngularDamping=true
    bUseGravityOverride=true
    gravityScale=0.35
    simulationSpace=RootRelative
    boundBone {boneName=dyn_bell_01_lf}
    componentAppliedLinearAccClamp=<0 0 0>
    gravityOverride {z=-980}
    physicsBodyDefinitions [{
      boundBone {boneName=dyn_bell_01_lf}
      boxExtents=<15 25 25>
      constraintSetup {angularLimitsMax=<0 0 5>, angularLimitsMin=<0 0 -5>, angularTarget=<1 0 0>}
      localJointOffset=<1 0 0>
     }]
    simSpaceSettings {simSpaceAngularAlpha=1}
   }
   n5 = Anim Dynamics - Bone: dyn_bell_02_lf!(ComponentPose=@n4, ExternalForce="0,0,0") -> (Pose) {
    alpha=0.7
    angularDampingOverride=1
    angularSpringConstant=60000
    bAngularSpring=true
    bOverrideAngularDamping=true
    bUseGravityOverride=true
    gravityScale=0.35
    simulationSpace=RootRelative
    boundBone {boneName=dyn_bell_02_lf}
    componentAppliedLinearAccClamp=<0 0 0>
    gravityOverride {z=-980}
    physicsBodyDefinitions [{
      boundBone {boneName=dyn_bell_02_lf}
      boxExtents=<15 25 25>
      constraintSetup {angularLimitsMax=<0 0 10>, angularLimitsMin=<0 0 -10>, angularTarget=<1 0 0>}
      localJointOffset=<1 0 0>
     }]
    simSpaceSettings {simSpaceAngularAlpha=1}
   }
   n6 = Component To Local(ComponentPose=@n5)
   n7 = Anim Dynamics - Bone: dyn_bell_01_rt(ComponentPose=Local To Component(LocalPose=Use cached pose 'Input_Pose'()), ExternalForce="0,0,0") {
    alpha=0.6
    angularDampingOverride=1
    angularSpringConstant=60000
    bAngularSpring=true
    bOverrideAngularDamping=true
    bUseGravityOverride=true
    gravityScale=0.35
    simulationSpace=RootRelative
    boundBone {boneName=dyn_bell_01_rt}
    componentAppliedLinearAccClamp=<0 0 0>
    gravityOverride {z=-980}
    physicsBodyDefinitions [{
      boundBone {boneName=dyn_bell_01_rt}
      boxExtents=<15 25 25>
      constraintSetup {angularLimitsMax=<0 0 5>, angularLimitsMin=<0 0 -5>, angularTarget=<1 0 0>}
      localJointOffset=<0 0 0>
     }]
    simSpaceSettings {simSpaceAngularAlpha=1}
   }
   n8 = Anim Dynamics - Bone: dyn_bell_02_rt(ComponentPose=@n7, ExternalForce="0,0,0") {
    alpha=0.6
    angularDampingOverride=1
    angularSpringConstant=60000
    bAngularSpring=true
    bOverrideAngularDamping=true
    bUseGravityOverride=true
    gravityScale=0.35
    simulationSpace=RootRelative
    boundBone {boneName=dyn_bell_02_rt}
    componentAppliedLinearAccClamp=<0 0 0>
    gravityOverride {z=-980}
    physicsBodyDefinitions [{
      boundBone {boneName=dyn_bell_02_rt}
      boxExtents=<15 25 25>
      constraintSetup {angularLimitsMax=<0 0 10>, angularLimitsMin=<0 0 -10>, angularTarget=<1 0 0>}
      localJointOffset=<0 0 0>
     }]
    simSpaceSettings {simSpaceAngularAlpha=1}
   }
   n9 = Anim Dynamics - Bone: dyn_bell_01_lf(ComponentPose=@n8, ExternalForce="0,0,0") {
    alpha=0.6
    angularDampingOverride=1
    angularSpringConstant=60000
    bAngularSpring=true
    bOverrideAngularDamping=true
    bUseGravityOverride=true
    gravityScale=0.35
    simulationSpace=RootRelative
    boundBone {boneName=dyn_bell_01_lf}
    componentAppliedLinearAccClamp=<0 0 0>
    gravityOverride {z=-980}
    physicsBodyDefinitions [{
      boundBone {boneName=dyn_bell_01_lf}
      boxExtents=<15 25 25>
      constraintSetup {angularLimitsMax=<0 0 5>, angularLimitsMin=<0 0 -5>, angularTarget=<1 0 0>}
      localJointOffset=<0 0 0>
     }]
    simSpaceSettings {simSpaceAngularAlpha=1}
   }
   n10 = Anim Dynamics - Bone: dyn_bell_02_lf(ComponentPose=@n9, ExternalForce="0,0,0") {
    alpha=0.6
    angularDampingOverride=1
    angularSpringConstant=60000
    bAngularSpring=true
    bOverrideAngularDamping=true
    bUseGravityOverride=true
    gravityScale=0.35
    simulationSpace=RootRelative
    boundBone {boneName=dyn_bell_02_lf}
    componentAppliedLinearAccClamp=<0 0 0>
    gravityOverride {z=-980}
    physicsBodyDefinitions [{
      boundBone {boneName=dyn_bell_02_lf}
      boxExtents=<15 25 25>
      constraintSetup {angularLimitsMax=<0 0 10>, angularLimitsMin=<0 0 -10>, angularTarget=<1 0 0>}
      localJointOffset=<0 0 0>
     }]
    simSpaceSettings {simSpaceAngularAlpha=1}
   }
   n11 = Component To Local(ComponentPose=@n10)
   n12 = Anim Dynamics - Bone: dyn_bell_01_rt!(ComponentPose=Local To Component(LocalPose=Use cached pose 'Input_Pose'()), Alpha:float=1, ExternalForce="0,0,0") -> (Pose) {
    angularDampingOverride=1
    angularSpringConstant=60000
    bAngularSpring=true
    bOverrideAngularDamping=true
    bUseGravityOverride=true
    gravityScale=0.35
    simulationSpace=RootRelative
    boundBone {boneName=dyn_bell_01_rt}
    componentLinearAccScale=<1 1 1>
    componentLinearVelScale=<1 1 1>
    gravityOverride {z=-980}
    physicsBodyDefinitions [{
      boundBone {boneName=dyn_bell_01_rt}
      boxExtents=<20 30 30>
      constraintSetup {angularLimitsMax=<0 0 10>, angularLimitsMin=<0 0 -10>, angularTarget=<1 0 0>}
      localJointOffset=<2 0 0>
     }]
   }
   n13 = Anim Dynamics - Bone: dyn_bell_02_rt!(ComponentPose=@n12, Alpha:float=1, ExternalForce="0,0,0") -> (Pose) {
    angularDampingOverride=1
    angularSpringConstant=60000
    bAngularSpring=true
    bOverrideAngularDamping=true
    bUseGravityOverride=true
    gravityScale=0.35
    simulationSpace=RootRelative
    boundBone {boneName=dyn_bell_02_rt}
    componentLinearAccScale=<1 1 1>
    componentLinearVelScale=<1 1 1>
    gravityOverride {z=-980}
    physicsBodyDefinitions [{
      boundBone {boneName=dyn_bell_02_rt}
      boxExtents=<20 30 30>
      constraintSetup {angularLimitsMax=<0 0 15>, angularLimitsMin=<0 0 -15>, angularTarget=<1 0 0>}
      localJointOffset=<2 0 0>
     }]
   }
   n14 = Anim Dynamics - Bone: dyn_bell_01_lf!(ComponentPose=@n13, Alpha:float=1, ExternalForce="0,0,0") -> (Pose) {
    angularDampingOverride=1
    angularSpringConstant=60000
    bAngularSpring=true
    bOverrideAngularDamping=true
    bUseGravityOverride=true
    gravityScale=0.35
    simulationSpace=RootRelative
    boundBone {boneName=dyn_bell_01_lf}
    componentLinearAccScale=<1 1 1>
    componentLinearVelScale=<1 1 1>
    gravityOverride {z=-980}
    physicsBodyDefinitions [{
      boundBone {boneName=dyn_bell_01_lf}
      boxExtents=<20 30 30>
      constraintSetup {angularLimitsMax=<0 0 10>, angularLimitsMin=<0 0 -10>, angularTarget=<1 0 0>}
      localJointOffset=<2 0 0>
     }]
   }
   n15 = Anim Dynamics - Bone: dyn_bell_02_lf!(ComponentPose=@n14, Alpha:float=1, ExternalForce="0,0,0") -> (Pose) {
    angularDampingOverride=1
    angularSpringConstant=60000
    bAngularSpring=true
    bOverrideAngularDamping=true
    bUseGravityOverride=true
    gravityScale=0.35
    simulationSpace=RootRelative
    boundBone {boneName=dyn_bell_02_lf}
    componentLinearAccScale=<1 1 1>
    componentLinearVelScale=<1 1 1>
    gravityOverride {z=-980}
    physicsBodyDefinitions [{
      boundBone {boneName=dyn_bell_02_lf}
      boxExtents=<20 30 30>
      constraintSetup {angularLimitsMax=<0 0 15>, angularLimitsMin=<0 0 -15>, angularTarget=<1 0 0>}
      localJointOffset=<2 0 0>
     }]
   }
   Output Pose(Result=Blend Poses by bool(BlendPose_0=Blend Poses by bool(BlendPose_0=@n11, BlendPose_1=@n6), BlendPose_1=Component To Local(ComponentPose=@n15)))
   Save cached pose 'Input_Pose'(Pose=Input Pose()) {cacheName=Input_Pose}
  ]}
 EventGraph {nodes [
   Event BlueprintUpdateAnimation()
   > n1 = Branch(Condition=Get bIsSkydiveFloating())
   > Set gravity_override(gravity_override="0,0,980")
   Set gravity_override(execute=@n1.else, gravity_override="0,0,-980")
   TryGetPawnOwner()
  ]}
]
variables [gravity_override {type=Vector}]
